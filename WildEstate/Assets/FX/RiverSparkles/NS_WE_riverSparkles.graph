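# Niagara system that scatters drifting, spinning sparkle particles over the rivers and lakes of the WildEstate map.
# GPU emitter sparkles_GPU merges the disabled lake and river emitters as two spawn groups, culled by a world-aligned river mask texture and bounds box.
# Landscape sampling or a fixed height band places particles; curl noise and wind move them horizontally, and camera distance scales size.
# Sparkles_Low is a simple burst emitter; the water-body-snapping CPU variant is disabled.
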NiagaraSystem /WildEstate/Assets/FX/RiverSparkles/NS_WE_riverSparkles
dependencies {
 usedDataInterfaces [
  /Script/Niagara.NiagaraDataInterfaceTexture
  &NiagaraDataInterfaceCurve
  &NiagaraDataInterfaceColorCurve
  &NiagaraDataInterfaceCamera
  /Script/Water.NiagaraDataInterfaceWater
  &NiagaraDataInterfaceCollisionQuery
  &NiagaraDataInterfaceLandscape
 ]
 usedDynamicInputs []
 usedModules [
  &self:SystemSpawnScript.SystemScriptSource.SystemScriptGraph.NiagaraNodeAssignment_0.SetVariables_#1
  &self:SystemSpawnScript.SystemScriptSource.SystemScriptGraph.NiagaraNodeAssignment_2.SetVariables_#2
  &SystemState
  &self:SystemSpawnScript.SystemScriptSource.SystemScriptGraph.NiagaraNodeAssignment_1.SetVariables_#3
 ]
}
emitters {
 Lake_sparkes {
  bEnabled=false
  simTarget=GPUComputeSim
  emitterUpdateScript [
   &EmitterState()
   &SpawnRate(SpawnRate=85, Use Spawn Probability=false)
  ]
  particleSpawnScript [
   &AddVelocity(Velocity=<0 50 0>)
   &InitializeParticle(
    Color=<rgba 0.5826 1 0.6842 1>
    Lifetime Max=3.25
    Lifetime Mode=::NewEnumerator1
    Sprite Rotation Mode=::NewEnumerator2
    Sprite Size=&Multiply_Vector2D_ByFloat(
     Float=&RandomRangeFloat(Minimum=40, Maximum=20)
     Vector2D=&Multiply_Vector2D_ByFloat(Float=0.7, Vector2D=<1 1>)
    )
    Sprite Size Mode=::NewEnumerator4
    Sprite UV Mode=::NewEnumerator1
    UsePositionOffset=false
    Write Material Random=true
   )
   /SetVariables(Particles.Terrain Offset:NiagaraFloat=&RandomRangeFloat(Minimum=1500, Maximum=1500)) {isSetParameters=true, module=SetVariables_#4}
   &ShapeLocation(
    Cylinder Radius=10000
    MaskSpawnGroup=false
    Override Seed=false
    Shape Primitive=::NewEnumerator1
    Surface Only=true
    Surface Only Band Thickness=1
    Use Endcaps In Surface Only Mode=false
    UseOwnerScale=true
   )
   &self:Sample_Landscape_01(Landscape Sample:NiagaraDataInterfaceLandscape=$System.Landscape Sample) {module=Sample_Landscape_01}
   &DynamicMaterialParameters(
    Index 0 Param 4=1
    Module.Index0Float0Write=true
    Module.Index0Float1Write=true
    Module.Index0Float2Write=true
    Module.Index0Float3Write=true
    Module.Index1Float0Write=true
    Module.Index1Float1Write=true
    Module.Index1Float2Write=true
    Module.Index1Float3Write=true
    Module.Index2Float0Write=true
    Module.Index2Float1Write=true
    Module.Index2Float2Write=true
    Module.Index2Float3Write=true
    Module.Index3Float0Write=true
    Module.Index3Float1Write=true
    Module.Index3Float2Write=true
    Module.Index3Float3Write=true
    Write Parameter Index 1=::NewEnumerator1
    Write Parameter Index 2=::NewEnumerator1
    Write Parameter Index 3=::NewEnumerator1
   )
   /SetVariables(Particles.Linear Color=<rgba 1 0 0 1>, Particles.WaveScale:Vector3f=<6 6 6>) {isSetParameters=true, module=SetVariables_#5}
   /SetVariables(Particles.Position=&Add_Vector(
     A=Linked() {linkedVariable {name=Particles.Position, type=NiagaraPosition}}
     B=&MakeVector(X=0, Y=0, Z=$Particles.Terrain Offset)
    )) {isSetParameters=true, module=SetVariables_#6}
   &ApplyInitialForces()
   &CameraOffset(Camera Offset Amount=10) {module=CameraOffset001}
   /Niagara/Modules/Update/Textures/WorldAlignedTextureSample(
    Center Position:Vector3f=<57325 -50000 1381.621>
    Texture=DataInterface() {propertyValues {
      Texture=&/T_WildEstateRivers
      TextureUserParameter {parameter {name=None, typeDefHandle {registeredTypeIndex=2}, varData []}}
     }}
    Texture Scale:Vector2f=<100000 100000>
    XY Vector Source Space:ENiagaraCoordinateSpace=::World
    XY Vector Target Space:ENiagaraCoordinateSpace=::World
   )
   &KillParticles(Kill Particles=&SetBoolByFloatComparison(
     A=&MakeFloatFromLinearColor(LinearColor=$Output.WorldAlignedTextureSample.SampledColor, Channel:ENiagaraFloatFromLinearColorOptions=::NewEnumerator1)
     B=0.5
     Comparison Type=::NewEnumerator6
    ))
   &KillParticlesInVolume(
    Box Size:Vector3f=<100000 100000 5000>
    Invert Volume=true
    Kill Shape=::NewEnumerator1
    Volume Origin:Vector3f=&TransformPosition(Position=<4065.4495 4264.7632 1381.6213>, Source Space=::World, Destination Space=::Simulation)
   ) {module=KillParticlesInVolume001}
   /SetVariables(Particles.Position=&self:ScratchDynamicInput_01(
     Position XY:NiagaraPosition=$Particles.Position
     Position Z:NiagaraPosition=&AddVectorToPosition(Position=<0 0 1102.1721>, Vector=&RandomRangeVector(Minimum=<0 0 0>, Maximum=<0 0 200>))
    )) {isSetParameters=true, module=SetVariables_#7}
  ]
  particleUpdateScript [
   &ParticleState()
   &SpriteRotationRate(Rotation Rate=&Multiply_Float(
     A=&RandomRangeFloat(Minimum=20, Maximum=15)
     B=&UniformAOrBFloat(A=-1, Evaluation Type=::NewEnumerator0, Recalculate Random Each Loop=true, Override Seed=false)
    )) {module=SpriteRotationRate001}
   &ScaleSpriteSize(
    Uniform Curve Index=&DistanceBetweenPositions(
     A=&ReturnCameraPosition(Coordinate Space:ENiagaraCoordinateSpace=::World, Camera Query:NiagaraDataInterfaceCamera=DataInterface())
     B=$Particles.Position
    )
    Uniform Curve Sprite Scale=DataInterface() {propertyValues {Curve {keys [<at 453.3277 0>,<at 603.3277 1>]}}}
   ) {module=ScaleSpriteSize001}
   &ScaleSpriteSize(
    Scale Sprite Size Mode=::NewEnumerator0
    Uniform Scale Factor=&FloatFromCurve(
     FloatCurve=$System.Curve for Floats
     CurveIndex=&Divide_Float(A=&DistanceBetweenPositions(A=$System.CamPos, B=$Particles.Position), B=10000)
    )
   ) {module=ScaleSpriteSize002}
   &CurlNoiseForce(
    Curl Noise Cone Mask Angle=25
    Mask Curl Noise=true
    Noise Frequency=&RandomRangeFloat(Minimum=50, Maximum=750)
    Noise Strength=&Multiply_Float(A=&RandomRangeFloat(Minimum=25, Maximum=100), B=1)
    Pan Noise=true
   )
   &WindForce(Use Depth Buffer=false, Use Landscape=false, Wind Speed=<0 25 0>)
   &Drag(Drag=0.1, Use Linear Drag=true, Use Rotational Drag=true)
   &SolveForcesAndVelocity(
    Acceleration Limit=5
    Clamp Velocity=true
    Force:Vector3f=&Multiply_Vector(A=$Transient.PhysicsForce, B=<1 1 0>)
    Limit Acceleration=true
    Speed Limit=25
    Velocity:Vector3f=&Multiply_Vector(A=$Particles.Velocity, B=<1 1 0>)
   )
   &CalculateAccurateVelocity()
   &CurlNoiseForce(
    Noise Frequency=&RandomRangeFloat(Minimum=0, Maximum=1.5)
    Noise Strength=&Multiply_Float(A=&RandomRangeFloat(Minimum=37.5, Maximum=75), B=1)
    Pan Noise=false
    Pan Noise Field=<0 0 2>
   ) {module=CurlNoiseForce001}
   &SolveForcesAndVelocity(
    Acceleration Limit=5
    Clamp Velocity=true
    Force:Vector3f=&Multiply_Vector(A=$Transient.PhysicsForce, B=<1 1 0>)
    Limit Acceleration=true
    Speed Limit=100
    Velocity:Vector3f=&Multiply_Vector(A=$Particles.Velocity, B=<1 1 0>)
   ) {module=SolveForcesAndVelocity001}
   /SetVariables(
    Particles.Linear Color=&MultiplyLinearColor(
     Linear Color=$Particles.Color
     Scale Factor=&MakeVector4_FromFloat(Float=&FloatFromCurve(
       FloatCurve=DataInterface() {propertyValues {Curve {keys [
           {interpMode=RCIM_Cubic, tangentMode=RCTM_User, time=0, value=0}
           {interpMode=RCIM_Cubic, tangentMode=RCTM_User, time=0.5, value=1}
           {tangentMode=RCTM_User, time=1, value=0}
          ]}}}
       Scale Curve=$System.ColorIntensityMult
      ))
    )
    Particles.Linear Color001:LinearColor=&MultiplyLinearColor(
     Linear Color=&ColorFromCurve(ColorCurve=DataInterface() {propertyValues {
        AlphaCurve {keys [<at 0 0 RCIM_Cubic>,<at 0.1218 1>,<at 1 0 RCIM_Cubic>]}
        BlueCurve {keys [<at 0 0.0688 RCIM_Cubic>,<at 1 1 RCIM_Cubic>]}
        GreenCurve {keys [<at 0 1 RCIM_Cubic>,<at 1 1 RCIM_Cubic>]}
        RedCurve {keys [<at 0 0 RCIM_Cubic>,<at 1 1 RCIM_Cubic>]}
       }})
     Scale Factor=&MakeVector4_FromFloat(Float=&FloatFromCurve(
       FloatCurve=DataInterface() {propertyValues {Curve {keys [<at 0 1>,<at 1 0>]}}}
       Scale Curve=&Multiply_Float(A=$System.ColorIntensityMult, B=4)
      ))
    )
   ) {isSetParameters=true, module=SetVariables_#8}
   &ScaleSpriteSize(Uniform Curve Sprite Scale=DataInterface() {propertyValues {Curve {
       defaultValue=0
       keys [
        {arriveTangent=9.0725, interpMode=RCIM_Cubic, time=0, value=0}
        <at 0.35 0.9847 RCIM_Cubic>
        {arriveTangent=-0.0405, interpMode=RCIM_Cubic, leaveTangent=-0.0405, tangentMode=RCTM_User, time=1, value=0}
       ]
      }}})
   &DynamicMaterialParameters(
    Index 0 Param 2=0.5
    Index 0 Param 4=1
    Index 1 Param 1=0.5
    Index 1 Param 3=2
    Index 2 Vector 1:Vector3f=<0 0.2 0>
    Module.Index0Float0Write=true
    Module.Index0Float1Write=true
    Module.Index0Float2Write=true
    Module.Index0Float3Write=true
    Module.Index1Float0Write=true
    Module.Index1Float1Write=true
    Module.Index1Float2Write=true
    Module.Index1Float3Write=true
    Module.Index2Float3Write=true
    Module.Index2VectorWrite=true
    Write Parameter Index 1=::NewEnumerator1
    Write Parameter Index 2=::NewEnumerator3
   ) {module=DynamicMaterialParameters001}
   /SetVariables(
    Particles.Color 2:LinearColor=<rgba 0 1 0 0.1>
    Particles.SpriteSize 2:Vector2f=&Multiply_Vector2D_ByFloat(
     Float=3
     Vector2D=&Vector2DFromFloat(Value=&Max_Float(
       A:NiagaraFloat=&MakeFloatFromVector2d(Vector 2D=$Particles.SpriteSize)
       B:NiagaraFloat=&MakeFloatFromVector2d(Vector 2D=$Particles.SpriteSize, Channel=::NewEnumerator1)
      ))
    )
   ) {isSetParameters=true, module=SetVariables_#9}
   /SetVariables(Particles.Vector 4:Vector4f=&MakeVector4(X=2, Y=1, Z=0, W=0)) {isSetParameters=true, module=SetVariables_#10}
   /SetVariables(Particles.SpriteRotation:NiagaraFloat=&FloatFromCurve(
     FloatCurve=DataInterface() {propertyValues {Curve {keys [<at 0 1>,<at 1 0>]}}}
     Scale Curve=&UniformAOrBFloat(A=-360, B=360, Evaluation Type=::NewEnumerator0, Recalculate Random Each Loop=true, Override Seed=false)
    )) {isSetParameters=true, module=SetVariables_#11}
  ]
  rendererClasses [&NiagaraSpriteRendererProperties, &NiagaraMeshRendererProperties]
  renderers [
   {rendererClass=&NiagaraSpriteRendererProperties}
   {rendererClass=&NiagaraSpriteRendererProperties, rendererIndex=1}
   {rendererClass=&NiagaraMeshRendererProperties, rendererIndex=2}
  ]
 }
 RiverSparkles {
  bEnabled=false
  simTarget=GPUComputeSim
  emitterUpdateScript [
   &EmitterState()
   &SpawnRate(SpawnRate=240, Use Spawn Probability=false)
  ]
  particleSpawnScript [
   &AddVelocity(Velocity=<0 50 0>)
   &InitializeParticle(
    Color=<rgba 0.5826 1 0.6842 1>
    Lifetime Max=3.25
    Lifetime Mode=::NewEnumerator1
    Sprite Rotation Mode=::NewEnumerator2
    Sprite Size=&Multiply_Vector2D_ByFloat(
     Float=&RandomRangeFloat(Minimum=40, Maximum=20)
     Vector2D=&Multiply_Vector2D_ByFloat(Float=0.7, Vector2D=<1 1>)
    )
    Sprite Size Mode=::NewEnumerator4
    Sprite UV Mode=::NewEnumerator1
    UsePositionOffset=false
    Write Material Random=true
   )
   /SetVariables(Particles.Terrain Offset:NiagaraFloat=&RandomRangeFloat(Minimum=150, Maximum=250)) {isSetParameters=true, module=SetVariables_#4}
   &ShapeLocation(
    Cylinder Radius=10000
    MaskSpawnGroup=false
    Override Seed=false
    Shape Primitive=::NewEnumerator1
    Surface Only=true
    Surface Only Band Thickness=1
    Use Endcaps In Surface Only Mode=false
    UseOwnerScale=true
   )
   &self:Sample_Landscape_01(Landscape Sample:NiagaraDataInterfaceLandscape=$System.Landscape Sample) {module=Sample_Landscape_01}
   &DynamicMaterialParameters(
    Index 0 Param 4=1
    Module.Index0Float0Write=true
    Module.Index0Float1Write=true
    Module.Index0Float2Write=true
    Module.Index0Float3Write=true
    Module.Index1Float0Write=true
    Module.Index1Float1Write=true
    Module.Index1Float2Write=true
    Module.Index1Float3Write=true
    Module.Index2Float0Write=true
    Module.Index2Float1Write=true
    Module.Index2Float2Write=true
    Module.Index2Float3Write=true
    Module.Index3Float0Write=true
    Module.Index3Float1Write=true
    Module.Index3Float2Write=true
    Module.Index3Float3Write=true
    Write Parameter Index 1=::NewEnumerator1
    Write Parameter Index 2=::NewEnumerator1
    Write Parameter Index 3=::NewEnumerator1
   )
   /SetVariables(Particles.Linear Color=<rgba 1 0 0 1>, Particles.WaveScale:Vector3f=<6 6 6>) {isSetParameters=true, module=SetVariables_#5}
   /SetVariables(Particles.Position=&Add_Vector(
     A=Linked() {linkedVariable {name=Particles.Position, type=NiagaraPosition}}
     B=&MakeVector(X=0, Y=0, Z=$Particles.Terrain Offset)
    )) {isSetParameters=true, module=SetVariables_#6}
   &ApplyInitialForces()
   &CameraOffset(Camera Offset Amount=10) {module=CameraOffset001}
   &WorldAlignedTextureSample(
    Center Position:Vector3f=<57325 -50000 1381.621>
    Texture=DataInterface() {propertyValues {
      Texture=&T_WildEstateRivers
      TextureUserParameter {parameter {name=None, typeDefHandle {registeredTypeIndex=2}, varData []}}
     }}
    Texture Scale:Vector2f=<100000 100000>
    XY Vector Source Space:ENiagaraCoordinateSpace=::World
    XY Vector Target Space:ENiagaraCoordinateSpace=::World
   )
   &KillParticles(Kill Particles=&SetBoolByFloatComparison(
     A=&MakeFloatFromLinearColor(LinearColor=$Output.WorldAlignedTextureSample.SampledColor, Channel:ENiagaraFloatFromLinearColorOptions=::NewEnumerator0)
     B=0.5
     Comparison Type=::NewEnumerator6
    ))
   &KillParticlesInVolume(
    Box Size:Vector3f=<100000 100000 5000>
    Invert Volume=true
    Kill Shape=::NewEnumerator1
    Volume Origin:Vector3f=&TransformPosition(Position=<4065.4495 4264.7632 1381.6213>, Source Space=::World, Destination Space=::Simulation)
   ) {module=KillParticlesInVolume001}
  ]
  particleUpdateScript [
   &self:Sample_Landscape_UPdate_01(Bias:NiagaraFloat=$Particles.Terrain Offset, New Landscape Sample:NiagaraDataInterfaceLandscape=$System.Landscape Sample) {module=Sample_Landscape_UPdate_01}
   &ParticleState()
   &SpriteRotationRate(Rotation Rate=&Multiply_Float(
     A=&RandomRangeFloat(Minimum=20, Maximum=15)
     B=&UniformAOrBFloat(A=-1, Evaluation Type=::NewEnumerator0, Recalculate Random Each Loop=true, Override Seed=false)
    )) {module=SpriteRotationRate001}
   &ScaleSpriteSize(
    Scale Sprite Size Mode=::NewEnumerator0
    Uniform Scale Factor=&FloatFromCurve(
     FloatCurve=$System.Curve for Floats
     CurveIndex=&Divide_Float(A=&DistanceBetweenPositions(A=$System.CamPos, B=$Particles.Position), B=10000)
    )
   ) {module=ScaleSpriteSize002}
   &ScaleSpriteSize(
    Uniform Curve Index=&DistanceBetweenPositions(A=$System.CamPos, B=$Particles.Position)
    Uniform Curve Sprite Scale=DataInterface() {propertyValues {Curve {keys [<at 453.3277 0>,<at 603.3277 1>]}}}
   ) {module=ScaleSpriteSize001}
   &CurlNoiseForce(
    Curl Noise Cone Mask Angle=25
    Mask Curl Noise=true
    Noise Frequency=&RandomRangeFloat(Minimum=50, Maximum=750)
    Noise Strength=&Multiply_Float(A=&RandomRangeFloat(Minimum=25, Maximum=100), B=1)
    Pan Noise=true
   )
   &WindForce(Use Depth Buffer=false, Use Landscape=false, Wind Speed=<0 25 0>)
   &Drag(Drag=0.1, Use Linear Drag=true, Use Rotational Drag=true)
   &SolveForcesAndVelocity(
    Acceleration Limit=5
    Clamp Velocity=true
    Force:Vector3f=&Multiply_Vector(A=$Transient.PhysicsForce, B=<1 1 0>)
    Limit Acceleration=true
    Speed Limit=25
    Velocity:Vector3f=&Multiply_Vector(A=$Particles.Velocity, B=<1 1 0>)
   )
   &CalculateAccurateVelocity()
   &CurlNoiseForce(
    Noise Frequency=&RandomRangeFloat(Minimum=0, Maximum=1.5)
    Noise Strength=&Multiply_Float(A=&RandomRangeFloat(Minimum=37.5, Maximum=75), B=1)
    Pan Noise=false
    Pan Noise Field=<0 0 2>
   ) {module=CurlNoiseForce001}
   &SolveForcesAndVelocity(
    Acceleration Limit=5
    Clamp Velocity=true
    Force:Vector3f=&Multiply_Vector(A=$Transient.PhysicsForce, B=<1 1 0>)
    Limit Acceleration=true
    Speed Limit=100
    Velocity:Vector3f=&Multiply_Vector(A=$Particles.Velocity, B=<1 1 0>)
   ) {module=SolveForcesAndVelocity001}
   /SetVariables(
    Particles.Linear Color=&MultiplyLinearColor(
     Linear Color=$Particles.Color
     Scale Factor=&MakeVector4_FromFloat(Float=&FloatFromCurve(
       FloatCurve=DataInterface() {propertyValues {Curve {keys [
           {interpMode=RCIM_Cubic, tangentMode=RCTM_User, time=0, value=0}
           {interpMode=RCIM_Cubic, tangentMode=RCTM_User, time=0.5, value=1}
           {tangentMode=RCTM_User, time=1, value=0}
          ]}}}
       Scale Curve=$System.ColorIntensityMult
      ))
    )
    Particles.Linear Color001:LinearColor=&MultiplyLinearColor(
     Linear Color=&ColorFromCurve(ColorCurve=DataInterface() {propertyValues {
        AlphaCurve {keys [<at 0 0 RCIM_Cubic>,<at 0.1218 1>,<at 1 0 RCIM_Cubic>]}
        BlueCurve {keys [<at 0 0.0688 RCIM_Cubic>,<at 1 1 RCIM_Cubic>]}
        GreenCurve {keys [<at 0 1 RCIM_Cubic>,<at 1 1 RCIM_Cubic>]}
        RedCurve {keys [<at 0 0 RCIM_Cubic>,<at 1 1 RCIM_Cubic>]}
       }})
     Scale Factor=&MakeVector4_FromFloat(Float=&FloatFromCurve(
       FloatCurve=DataInterface() {propertyValues {Curve {keys [<at 0 1>,<at 1 0>]}}}
       Scale Curve=&Multiply_Float(A=$System.ColorIntensityMult, B=4)
      ))
    )
   ) {isSetParameters=true, module=SetVariables_#8}
   &ScaleSpriteSize(Uniform Curve Sprite Scale=DataInterface() {propertyValues {Curve {
       defaultValue=0
       keys [
        {arriveTangent=9.0725, interpMode=RCIM_Cubic, time=0, value=0}
        <at 0.35 0.9847 RCIM_Cubic>
        {arriveTangent=-0.0405, interpMode=RCIM_Cubic, leaveTangent=-0.0405, tangentMode=RCTM_User, time=1, value=0}
       ]
      }}})
   &DynamicMaterialParameters(
    Index 0 Param 2=0.5
    Index 0 Param 4=1
    Index 1 Param 1=0.5
    Index 1 Param 3=2
    Index 2 Vector 1:Vector3f=<0 0.2 0>
    Module.Index0Float0Write=true
    Module.Index0Float1Write=true
    Module.Index0Float2Write=true
    Module.Index0Float3Write=true
    Module.Index1Float0Write=true
    Module.Index1Float1Write=true
    Module.Index1Float2Write=true
    Module.Index1Float3Write=true
    Module.Index2Float3Write=true
    Module.Index2VectorWrite=true
    Write Parameter Index 1=::NewEnumerator1
    Write Parameter Index 2=::NewEnumerator3
   ) {module=DynamicMaterialParameters001}
   /SetVariables(
    Particles.Color 2:LinearColor=<rgba 0 1 0 0.1>
    Particles.SpriteSize 2:Vector2f=&Multiply_Vector2D_ByFloat(
     Float=3
     Vector2D=&Vector2DFromFloat(Value=&Max_Float(
       A:NiagaraFloat=&MakeFloatFromVector2d(Vector 2D=$Particles.SpriteSize)
       B:NiagaraFloat=&MakeFloatFromVector2d(Vector 2D=$Particles.SpriteSize, Channel=::NewEnumerator1)
      ))
    )
   ) {isSetParameters=true, module=SetVariables_#9}
   /SetVariables(Particles.Vector 4:Vector4f=&MakeVector4(X=2, Y=1, Z=0, W=0)) {isSetParameters=true, module=SetVariables_#10}
   /SetVariables(Particles.SpriteRotation:NiagaraFloat=&FloatFromCurve(
     FloatCurve=DataInterface() {propertyValues {Curve {keys [<at 0 1>,<at 1 0>]}}}
     Scale Curve=&UniformAOrBFloat(A=-360, B=360, Evaluation Type=::NewEnumerator0, Recalculate Random Each Loop=true, Override Seed=false)
    )) {isSetParameters=true}
  ]
  rendererClasses [&NiagaraSpriteRendererProperties, &NiagaraMeshRendererProperties]
  renderers [
   {rendererClass=&NiagaraSpriteRendererProperties}
   {rendererClass=&NiagaraSpriteRendererProperties, rendererIndex=1}
   {rendererClass=&NiagaraMeshRendererProperties, rendererIndex=2}
  ]
 }
 Sparkles_CPU {
  bEnabled=false
  emitterUpdateScript [
   &EmitterState()
   &SpawnRate(SpawnRate=85, Use Spawn Probability=false)
  ]
  particleSpawnScript [
   &AddVelocity(Velocity=<0 50 0>)
   &InitializeParticle(
    Color=<rgba 0.5826 1 0.6842 1>
    Lifetime Max=3.25
    Lifetime Mode=::NewEnumerator1
    Sprite Rotation Mode=::NewEnumerator2
    Sprite Size=&Multiply_Vector2D_ByFloat(
     Float=&RandomRangeFloat(Minimum=40, Maximum=20)
     Vector2D=&Multiply_Vector2D_ByFloat(Float=0.7, Vector2D=<1 1>)
    )
    Sprite Size Mode=::NewEnumerator4
    UsePositionOffset=false
    Write Material Random=true
   )
   /SetVariables(Particles.Terrain Offset:NiagaraFloat=&MaskFloatBySpawnGroup(Masked Float=1500, Passthrough Float=&RandomRangeFloat(Minimum=150, Maximum=250))) {isSetParameters=true, module=SetVariables_#4}
   &ShapeLocation(
    Cylinder Radius=10000
    MaskSpawnGroup=false
    Override Seed=false
    Shape Primitive=::NewEnumerator1
    Surface Only=true
    Surface Only Band Thickness=1
    Use Endcaps In Surface Only Mode=false
    UseOwnerScale=true
   )
   &DynamicMaterialParameters(
    Index 0 Param 4=1
    Module.Index0Float0Write=true
    Module.Index0Float1Write=true
    Module.Index0Float2Write=true
    Module.Index0Float3Write=true
    Module.Index1Float0Write=true
    Module.Index1Float1Write=true
    Module.Index1Float2Write=true
    Module.Index1Float3Write=true
    Module.Index2Float0Write=true
    Module.Index2Float1Write=true
    Module.Index2Float2Write=true
    Module.Index2Float3Write=true
    Module.Index3Float0Write=true
    Module.Index3Float1Write=true
    Module.Index3Float2Write=true
    Module.Index3Float3Write=true
    Write Parameter Index 1=::NewEnumerator1
    Write Parameter Index 2=::NewEnumerator1
    Write Parameter Index 3=::NewEnumerator1
   )
   /SetVariables(Particles.Linear Color=<rgba 1 0 0 1>, Particles.WaveScale:Vector3f=<6 6 6>) {isSetParameters=true, module=SetVariables_#5}
   /SetVariables(Particles.Position=&Add_Vector(
     A=Linked() {linkedVariable {name=Particles.Position, type=NiagaraPosition}}
     B=&MakeVector(X=0, Y=0, Z=$Particles.Terrain Offset)
    )) {isSetParameters=true, module=SetVariables_#6}
   &ApplyInitialForces()
   &CameraOffset(Camera Offset Amount=10) {module=CameraOffset001}
   &KillParticlesInVolume(
    Box Size:Vector3f=<100000 100000 5000>
    Invert Volume=true
    Kill Shape=::NewEnumerator1
    Volume Origin:Vector3f=&TransformPosition(Position=<4065.4495 4264.7632 1381.6213>, Source Space=::World, Destination Space=::Simulation)
   ) {module=KillParticlesInVolume001}
   /SetVariables(Particles.Position=/Niagara/DynamicInputs/Helpers/MaskPosition(
     Mask Source:NiagaraInt32=0
     Mask Parameter:NiagaraInt32=?
     Mask:NiagaraInt32=0
     Masked Position:NiagaraPosition=&self:ScratchDynamicInput_01(
      Position XY:NiagaraPosition=$Particles.Position
      Position Z:NiagaraPosition=&AddVectorToPosition(Position=<0 0 1102.1721>, Vector=&RandomRangeVector(Minimum=<0 0 0>, Maximum=<0 0 200>))
     )
     Passthrough Position:NiagaraPosition=$Particles.Position
    )) {enabled=false, isSetParameters=true, module=SetVariables_#7}
  ]
  particleUpdateScript [
   &ParticleState()
   /Game/Effects/Niagara/Modules/Water/WaterBodyLocation(
    Set Water Particle Displacement Material Dynamic Parameters=false
    Snap Position to Water Position=true
    WaterBodyActor=DataInterface() {propertyValues {
      SourceBodyComponent=/WildEstate/Maps/WildEstate_Terrain:PersistentLevel.WildEstate_Waterbody_Ocean_Parent_C_UAID_387C764928FB356402_1331878425
      bEvaluateSystemDepth=true
      bEvaluateSystemDepthPerFrame=true
      bFindWaterBodyOnSpawn=true
     }}
   )
   &self:CollisionSurfaceTest(Lower End=-150, SkipTrace:NiagaraBool=&Boolean_NotOperation(Bool=$Transient.FirstFrame), Upper Start=3500) {module=CollisionSurfaceTest}
   &KillParticles(Kill Particles=&Boolean_AndOperation(Bool1=$Output.ParticleState.FirstFrame, Bool2=$Particles.CollisionValid)) {module=KillParticles001}
   &ScaleSpriteSize(
    Uniform Curve Index=&DistanceBetweenPositions(
     A=&ReturnCameraPosition(Coordinate Space:ENiagaraCoordinateSpace=::World, Camera Query:NiagaraDataInterfaceCamera=DataInterface())
     B=$Particles.Position
    )
    Uniform Curve Sprite Scale=DataInterface() {propertyValues {Curve {keys [<at 453.3277 0>,<at 603.3277 1>]}}}
   ) {module=ScaleSpriteSize001}
   &ScaleSpriteSize(
    Scale Sprite Size Mode=::NewEnumerator0
    Uniform Scale Factor=&FloatFromCurve(
     FloatCurve=$System.Curve for Floats
     CurveIndex=&Divide_Float(A=&DistanceBetweenPositions(A=$System.CamPos, B=$Particles.Position), B=10000)
    )
   ) {module=ScaleSpriteSize002}
   &SpriteRotationRate(Rotation Rate=&Multiply_Float(
     A=&RandomRangeFloat(Minimum=20, Maximum=15)
     B=&UniformAOrBFloat(A=-1, Evaluation Type=::NewEnumerator0, Recalculate Random Each Loop=true, Override Seed=false)
    )) {module=SpriteRotationRate001}
   &CurlNoiseForce(
    Curl Noise Cone Mask Angle=25
    Mask Curl Noise=true
    Noise Frequency=&RandomRangeFloat(Minimum=50, Maximum=750)
    Noise Strength=&Multiply_Float(A=&RandomRangeFloat(Minimum=25, Maximum=100), B=1)
    Pan Noise=true
   )
   &WindForce(Wind Speed=<0 25 0>)
   &Drag(Drag=0.1, Use Linear Drag=true, Use Rotational Drag=true)
   &SolveForcesAndVelocity(
    Acceleration Limit=5
    Clamp Velocity=true
    Force:Vector3f=&Multiply_Vector(A=$Transient.PhysicsForce, B=<1 1 0>)
    Limit Acceleration=true
    Speed Limit=25
    Velocity:Vector3f=&Multiply_Vector(A=$Particles.Velocity, B=<1 1 0>)
   )
   &CalculateAccurateVelocity()
   &CurlNoiseForce(
    Noise Frequency=&RandomRangeFloat(Minimum=0, Maximum=1.5)
    Noise Strength=&Multiply_Float(A=&RandomRangeFloat(Minimum=37.5, Maximum=75), B=1)
    Pan Noise=false
    Pan Noise Field=<0 0 2>
   ) {module=CurlNoiseForce001}
   &SolveForcesAndVelocity(
    Acceleration Limit=5
    Clamp Velocity=true
    Force:Vector3f=&Multiply_Vector(A=$Transient.PhysicsForce, B=<1 1 0>)
    Limit Acceleration=true
    Speed Limit=100
    Velocity:Vector3f=&Multiply_Vector(A=$Particles.Velocity, B=<1 1 0>)
   ) {module=SolveForcesAndVelocity001}
   /SetVariables(
    Particles.Linear Color=&MultiplyLinearColor(
     Linear Color=$Particles.Color
     Scale Factor=&MakeVector4_FromFloat(Float=&FloatFromCurve(
       FloatCurve=DataInterface() {propertyValues {Curve {keys [
           {interpMode=RCIM_Cubic, tangentMode=RCTM_User, time=0, value=0}
           {interpMode=RCIM_Cubic, tangentMode=RCTM_User, time=0.5, value=1}
           {tangentMode=RCTM_User, time=1, value=0}
          ]}}}
       Scale Curve=$System.ColorIntensityMult
      ))
    )
    Particles.Linear Color001:LinearColor=&MultiplyLinearColor(
     Linear Color=&ColorFromCurve(ColorCurve=DataInterface() {propertyValues {
        AlphaCurve {keys [<at 0 0 RCIM_Cubic>,<at 0.1218 1>,<at 1 0 RCIM_Cubic>]}
        BlueCurve {keys [<at 0 0.0688 RCIM_Cubic>,<at 1 1 RCIM_Cubic>]}
        GreenCurve {keys [<at 0 1 RCIM_Cubic>,<at 1 1 RCIM_Cubic>]}
        RedCurve {keys [<at 0 0 RCIM_Cubic>,<at 1 1 RCIM_Cubic>]}
       }})
     Scale Factor=&MakeVector4_FromFloat(Float=&FloatFromCurve(
       FloatCurve=DataInterface() {propertyValues {Curve {keys [<at 0 1>,<at 1 0>]}}}
       Scale Curve=&Multiply_Float(A=$System.ColorIntensityMult, B=4)
      ))
    )
   ) {isSetParameters=true, module=SetVariables_#8}
   &ScaleSpriteSize(Uniform Curve Sprite Scale=DataInterface() {propertyValues {Curve {
       defaultValue=0
       keys [
        {arriveTangent=9.0725, interpMode=RCIM_Cubic, time=0, value=0}
        <at 0.35 0.9847 RCIM_Cubic>
        {arriveTangent=-0.0405, interpMode=RCIM_Cubic, leaveTangent=-0.0405, tangentMode=RCTM_User, time=1, value=0}
       ]
      }}})
   &DynamicMaterialParameters(
    Index 0 Param 2=0.5
    Index 0 Param 4=1
    Index 1 Param 1=0.5
    Index 1 Param 3=2
    Index 2 Vector 1:Vector3f=<0 0.2 0>
    Module.Index0Float0Write=true
    Module.Index0Float1Write=true
    Module.Index0Float2Write=true
    Module.Index0Float3Write=true
    Module.Index1Float0Write=true
    Module.Index1Float1Write=true
    Module.Index1Float2Write=true
    Module.Index1Float3Write=true
    Module.Index2Float3Write=true
    Module.Index2VectorWrite=true
    Write Parameter Index 1=::NewEnumerator1
    Write Parameter Index 2=::NewEnumerator3
   ) {module=DynamicMaterialParameters001}
   /SetVariables(
    Particles.Color 2:LinearColor=<rgba 0 1 0 0.1>
    Particles.SpriteSize 2:Vector2f=&Multiply_Vector2D_ByFloat(
     Float=3
     Vector2D=&Vector2DFromFloat(Value=&Max_Float(
       A:NiagaraFloat=&MakeFloatFromVector2d(Vector 2D=$Particles.SpriteSize)
       B:NiagaraFloat=&MakeFloatFromVector2d(Vector 2D=$Particles.SpriteSize, Channel=::NewEnumerator1)
      ))
    )
   ) {isSetParameters=true, module=SetVariables_#9}
   /SetVariables(Particles.Vector 4:Vector4f=&MakeVector4(X=2, Y=1, Z=0, W=0)) {isSetParameters=true, module=SetVariables_#10}
   /SetVariables(Particles.SpriteRotation:NiagaraFloat=&FloatFromCurve(
     FloatCurve=DataInterface() {propertyValues {Curve {keys [<at 0 1>,<at 1 0>]}}}
     Scale Curve=&UniformAOrBFloat(A=-360, B=360, Evaluation Type=::NewEnumerator0, Recalculate Random Each Loop=true, Override Seed=false)
    )) {isSetParameters=true, module=SetVariables_#11}
  ]
  rendererClasses [&NiagaraSpriteRendererProperties]
  renderers [
   {rendererClass=&NiagaraSpriteRendererProperties}
   {rendererClass=&NiagaraSpriteRendererProperties, rendererIndex=1}
  ]
 }
 Sparkles_Low {
  emitterUpdateScript [
   &EmitterState()
   &SpawnBurst_Instantaneous(Use Spawn Probability=false, UseLoopCountLimit=false)
  ]
  particleSpawnScript [&InitializeParticle(
    Color Mode=::NewEnumerator0
    Lifetime=5
    Position=<0 0 150>
    Position Mode=::NewEnumerator1
    UsePositionOffset=false
    Write Material Random=true
   )]
  particleUpdateScript [&ParticleState(Kill Particles When Lifetime Has Elapsed=false, Let Infinitely Lived Particles Die When Emitter Deactivates=true)]
  rendererClasses [&NiagaraSpriteRendererProperties, &NiagaraMeshRendererProperties]
  renderers [
   {rendererClass=&NiagaraSpriteRendererProperties}
   {rendererClass=&NiagaraMeshRendererProperties, rendererIndex=1}
  ]
 }
 sparkles_GPU {
  simTarget=GPUComputeSim
  emitterUpdateScript [
   &EmitterState()
   &SpawnRate(SpawnRate=85, Use Spawn Probability=false)
   &SpawnRate(SpawnGroup=1, SpawnRate=240, Use Spawn Probability=false) {module=SpawnRate001}
  ]
  particleSpawnScript [
   &AddVelocity(Velocity=<0 50 0>)
   &InitializeParticle(
    Color=<rgba 0.583 1 0.6842 1>
    Lifetime Max=3.25
    Lifetime Mode=::NewEnumerator1
    Sprite Rotation Mode=::NewEnumerator2
    Sprite Size=&Multiply_Vector2D_ByFloat(
     Float=&RandomRangeFloat(Minimum=40, Maximum=20)
     Vector2D=&Multiply_Vector2D_ByFloat(Float=0.7, Vector2D=<1 1>)
    )
    Sprite Size Mode=::NewEnumerator4
    Sprite UV Mode=::NewEnumerator1
    UsePositionOffset=false
    Write Material Random=true
   )
   /SetVariables(Particles.Terrain Offset:NiagaraFloat=&MaskFloatBySpawnGroup(Masked Float=1500, Passthrough Float=&RandomRangeFloat(Minimum=150, Maximum=250))) {isSetParameters=true, module=SetVariables_#4}
   &ShapeLocation(
    Cylinder Radius=10000
    MaskSpawnGroup=false
    Override Seed=false
    Shape Primitive=::NewEnumerator1
    Surface Only=true
    Surface Only Band Thickness=1
    Use Endcaps In Surface Only Mode=false
    UseOwnerScale=true
   )
   &self:Sample_Landscape_01(Landscape Sample:NiagaraDataInterfaceLandscape=$System.Landscape Sample) {module=Sample_Landscape_01}
   &DynamicMaterialParameters(
    Index 0 Param 4=1
    Module.Index0Float0Write=true
    Module.Index0Float1Write=true
    Module.Index0Float2Write=true
    Module.Index0Float3Write=true
    Module.Index1Float0Write=true
    Module.Index1Float1Write=true
    Module.Index1Float2Write=true
    Module.Index1Float3Write=true
    Module.Index2Float0Write=true
    Module.Index2Float1Write=true
    Module.Index2Float2Write=true
    Module.Index2Float3Write=true
    Module.Index3Float0Write=true
    Module.Index3Float1Write=true
    Module.Index3Float2Write=true
    Module.Index3Float3Write=true
    Write Parameter Index 1=::NewEnumerator1
    Write Parameter Index 2=::NewEnumerator1
    Write Parameter Index 3=::NewEnumerator1
   )
   /SetVariables(Particles.Linear Color=<rgba 1 0 0 1>, Particles.WaveScale:Vector3f=<6 6 6>) {isSetParameters=true, module=SetVariables_#5}
   /SetVariables(Particles.Position=&Add_Vector(
     A=Linked() {linkedVariable {name=Particles.Position, type=NiagaraPosition}}
     B=&MakeVector(X=0, Y=0, Z=$Particles.Terrain Offset)
    )) {isSetParameters=true, module=SetVariables_#6}
   &ApplyInitialForces()
   &CameraOffset(Camera Offset Amount=10) {module=CameraOffset001}
   &WorldAlignedTextureSample(
    Center Position:Vector3f=<57325 -50000 1381.621>
    Texture=DataInterface() {propertyValues {
      Texture=&T_WildEstateRivers
      TextureUserParameter {parameter {name=None, typeDefHandle {registeredTypeIndex=2}, varData []}}
     }}
    Texture Scale:Vector2f=<100000 100000>
    XY Vector Source Space:ENiagaraCoordinateSpace=::World
    XY Vector Target Space:ENiagaraCoordinateSpace=::World
   )
   &KillParticles(Kill Particles=&SetBoolByFloatComparison(
     A=&MaskFloatBySpawnGroup(
      Masked Float=&MakeFloatFromLinearColor(LinearColor=$Output.WorldAlignedTextureSample.SampledColor, Channel:ENiagaraFloatFromLinearColorOptions=::NewEnumerator1)
      Passthrough Float=&MakeFloatFromLinearColor(LinearColor=$Output.WorldAlignedTextureSample.SampledColor, Channel:ENiagaraFloatFromLinearColorOptions=::NewEnumerator0)
     )
     B=0.5
     Comparison Type=::NewEnumerator6
    ))
   &KillParticlesInVolume(
    Box Size:Vector3f=<100000 100000 5000>
    Invert Volume=true
    Kill Shape=::NewEnumerator1
    Volume Origin:Vector3f=&TransformPosition(Position=<4065.4495 4264.7632 1381.6213>, Source Space=::World, Destination Space=::Simulation)
   ) {module=KillParticlesInVolume001}
   /SetVariables(Particles.Position=&MaskPosition(
     Mask Source:NiagaraInt32=0
     Mask Parameter:NiagaraInt32=?
     Mask:NiagaraInt32=0
     Masked Position:NiagaraPosition=&self:ScratchDynamicInput_01(
      Position XY:NiagaraPosition=$Particles.Position
      Position Z:NiagaraPosition=&AddVectorToPosition(Position=<0 0 1102.1721>, Vector=&RandomRangeVector(Minimum=<0 0 0>, Maximum=<0 0 200>))
     )
     Passthrough Position:NiagaraPosition=$Particles.Position
    )) {isSetParameters=true, module=SetVariables_#7}
  ]
  particleUpdateScript [
   &self:Sample_Landscale(
    Bias:NiagaraFloat=$Particles.Terrain Offset
    New Landscape Sample:NiagaraDataInterfaceLandscape=$System.Landscape Sample
    Position:NiagaraPosition=$Particles.Position
    Use:NiagaraBool=/Niagara/DynamicInputs/Helpers/MaskBool(
     Mask Source:NiagaraInt32=0
     Mask Parameter:NiagaraInt32=?
     Mask:NiagaraInt32=0
     Masked Bool:NiagaraBool=false
     Passthrough Bool:NiagaraBool=false
    )
   ) {module=Sample_Landscale}
   &ParticleState()
   &ScaleSpriteSize(
    Uniform Curve Index=&DistanceBetweenPositions(
     A=&ReturnCameraPosition(Coordinate Space:ENiagaraCoordinateSpace=::World, Camera Query:NiagaraDataInterfaceCamera=DataInterface())
     B=$Particles.Position
    )
    Uniform Curve Sprite Scale=DataInterface() {propertyValues {Curve {keys [<at 453.3277 0>,<at 603.3277 1>]}}}
   ) {module=ScaleSpriteSize001}
   &ScaleSpriteSize(
    Scale Sprite Size Mode=::NewEnumerator0
    Uniform Scale Factor=&FloatFromCurve(
     FloatCurve=$System.Curve for Floats
     CurveIndex=&Divide_Float(A=&DistanceBetweenPositions(A=$System.CamPos, B=$Particles.Position), B=10000)
    )
   ) {module=ScaleSpriteSize002}
   &SpriteRotationRate(Rotation Rate=&Multiply_Float(
     A=&RandomRangeFloat(Minimum=20, Maximum=15)
     B=&UniformAOrBFloat(A=-1, Evaluation Type=::NewEnumerator0, Recalculate Random Each Loop=true, Override Seed=false)
    )) {module=SpriteRotationRate001}
   &CurlNoiseForce(
    Curl Noise Cone Mask Angle=25
    Mask Curl Noise=true
    Noise Frequency=&RandomRangeFloat(Minimum=50, Maximum=750)
    Noise Strength=&Multiply_Float(A=&RandomRangeFloat(Minimum=25, Maximum=100), B=1)
    Pan Noise=true
   )
   &WindForce(Use Depth Buffer=false, Use Landscape=false, Wind Speed=<0 25 0>)
   &Drag(Drag=0.1, Use Linear Drag=true, Use Rotational Drag=true)
   &SolveForcesAndVelocity(
    Acceleration Limit=5
    Clamp Velocity=true
    Force:Vector3f=&Multiply_Vector(A=$Transient.PhysicsForce, B=<1 1 0>)
    Limit Acceleration=true
    Speed Limit=25
    Velocity:Vector3f=&Multiply_Vector(A=$Particles.Velocity, B=<1 1 0>)
   )
   &CalculateAccurateVelocity()
   &CurlNoiseForce(
    Noise Frequency=&RandomRangeFloat(Minimum=0, Maximum=1.5)
    Noise Strength=&Multiply_Float(A=&RandomRangeFloat(Minimum=37.5, Maximum=75), B=1)
    Pan Noise=false
    Pan Noise Field=<0 0 2>
   ) {module=CurlNoiseForce001}
   &SolveForcesAndVelocity(
    Acceleration Limit=5
    Clamp Velocity=true
    Force:Vector3f=&Multiply_Vector(A=$Transient.PhysicsForce, B=<1 1 0>)
    Limit Acceleration=true
    Speed Limit=100
    Velocity:Vector3f=&Multiply_Vector(A=$Particles.Velocity, B=<1 1 0>)
   ) {module=SolveForcesAndVelocity001}
   /SetVariables(
    Particles.Linear Color=&MultiplyLinearColor(
     Linear Color=$Particles.Color
     Scale Factor=&MakeVector4_FromFloat(Float=&FloatFromCurve(
       FloatCurve=DataInterface() {propertyValues {Curve {keys [
           {interpMode=RCIM_Cubic, tangentMode=RCTM_User, time=0, value=0}
           {interpMode=RCIM_Cubic, tangentMode=RCTM_User, time=0.5, value=1}
           {tangentMode=RCTM_User, time=1, value=0}
          ]}}}
       Scale Curve=$System.ColorIntensityMult
      ))
    )
    Particles.Linear Color001:LinearColor=&MultiplyLinearColor(
     Linear Color=&ColorFromCurve(ColorCurve=DataInterface() {propertyValues {
        AlphaCurve {keys [<at 0 0 RCIM_Cubic>,<at 0.1218 1>,<at 1 0 RCIM_Cubic>]}
        BlueCurve {keys [<at 0 0.0688 RCIM_Cubic>,<at 1 1 RCIM_Cubic>]}
        GreenCurve {keys [<at 0 1 RCIM_Cubic>,<at 1 1 RCIM_Cubic>]}
        RedCurve {keys [<at 0 0 RCIM_Cubic>,<at 1 1 RCIM_Cubic>]}
       }})
     Scale Factor=&MakeVector4_FromFloat(Float=&FloatFromCurve(
       FloatCurve=DataInterface() {propertyValues {Curve {keys [<at 0 1>,<at 1 0>]}}}
       Scale Curve=&Multiply_Float(A=$System.ColorIntensityMult, B=4)
      ))
    )
   ) {isSetParameters=true, module=SetVariables_#8}
   &ScaleSpriteSize(Uniform Curve Sprite Scale=DataInterface() {propertyValues {Curve {
       defaultValue=0
       keys [
        {arriveTangent=9.0725, interpMode=RCIM_Cubic, time=0, value=0}
        <at 0.35 0.9847 RCIM_Cubic>
        {arriveTangent=-0.0405, interpMode=RCIM_Cubic, leaveTangent=-0.0405, tangentMode=RCTM_User, time=1, value=0}
       ]
      }}})
   &DynamicMaterialParameters(
    Index 0 Param 2=0.5
    Index 0 Param 4=1
    Index 1 Param 1=0.5
    Index 1 Param 3=2
    Index 2 Vector 1:Vector3f=<0 0.2 0>
    Module.Index0Float0Write=true
    Module.Index0Float1Write=true
    Module.Index0Float2Write=true
    Module.Index0Float3Write=true
    Module.Index1Float0Write=true
    Module.Index1Float1Write=true
    Module.Index1Float2Write=true
    Module.Index1Float3Write=true
    Module.Index2Float3Write=true
    Module.Index2VectorWrite=true
    Write Parameter Index 1=::NewEnumerator1
    Write Parameter Index 2=::NewEnumerator3
   ) {module=DynamicMaterialParameters001}
   /SetVariables(
    Particles.Color 2:LinearColor=<rgba 0 1 0 0.1>
    Particles.SpriteSize 2:Vector2f=&Multiply_Vector2D_ByFloat(
     Float=3
     Vector2D=&Vector2DFromFloat(Value=&Max_Float(
       A:NiagaraFloat=&MakeFloatFromVector2d(Vector 2D=$Particles.SpriteSize)
       B:NiagaraFloat=&MakeFloatFromVector2d(Vector 2D=$Particles.SpriteSize, Channel=::NewEnumerator1)
      ))
    )
   ) {isSetParameters=true, module=SetVariables_#9}
   /SetVariables(Particles.Vector 4:Vector4f=&MakeVector4(X=2, Y=1, Z=0, W=0)) {isSetParameters=true, module=SetVariables_#10}
   /SetVariables(Particles.SpriteRotation:NiagaraFloat=&FloatFromCurve(
     FloatCurve=DataInterface() {propertyValues {Curve {keys [<at 0 1>,<at 1 0>]}}}
     Scale Curve=&UniformAOrBFloat(A=-360, B=360, Evaluation Type=::NewEnumerator0, Recalculate Random Each Loop=true, Override Seed=false)
    )) {isSetParameters=true, module=SetVariables_#11}
  ]
  rendererClasses [&NiagaraSpriteRendererProperties]
  renderers [
   {rendererClass=&NiagaraSpriteRendererProperties}
   {rendererClass=&NiagaraSpriteRendererProperties, rendererIndex=1}
  ]
 }
}
summary {
 systemName=NS_WE_riverSparkles
 emitters [
  {
   bEnabled=false
   emitterName=RiverSparkles
   simTarget=GPUComputeSim
   rendererClasses [&NiagaraSpriteRendererProperties, &NiagaraMeshRendererProperties]
  }
  {
   bEnabled=false
   emitterName=Lake_sparkes
   simTarget=GPUComputeSim
   rendererClasses [&NiagaraSpriteRendererProperties, &NiagaraMeshRendererProperties]
  }
  {emitterName=sparkles_GPU, simTarget=GPUComputeSim, rendererClasses [&NiagaraSpriteRendererProperties]}
  {bEnabled=false, emitterName=Sparkles_CPU, rendererClasses [&NiagaraSpriteRendererProperties]}
  {emitterName=Sparkles_Low, rendererClasses [&NiagaraSpriteRendererProperties, &NiagaraMeshRendererProperties]}
 ]
 userVariables []
}
