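# Unlit translucent Niagara sprite material for river sparkle particles.
# A sparkle texture's red channel drives a blend between two dynamic-parameter colors and particle color, passed through an emissive adjustment function.
# Opacity adds the thresholded texture to a powered radial sphere mask, then applies camera-distance and depth fades.
# The star Shape blend has identical inputs, and Rad feeds both mask radius and hardness, leaving Hardness unused.
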Material /WildEstate/Assets/FX/RiverSparkles/RiverSparles
graph {
 properties {
  bAutomaticallySetUsageInEditor=false
  bUsedWithNiagaraSprites=true
  blendMode=BLEND_Translucent
  shadingModel=MSM_Unlit
  usedShadingModels=Unlit
 }
 nodes [
  n1 = TextureSample() {texture=&/T_River_Sparkles_basecolor}
  n2 = DynamicParameter() {defaultValue=<rgba 0 1 0 1>, paramNames [Color2 R, Color2 G, Color2 B, star Shape]}
  n3 = LinearInterpolate(A=@n1.R, B=@n1.R, Alpha=@n2.star Shape)
  n4 = SmoothStep(Value=@n3) {constMax=0.01}
  n5 = TextureCoordinate()
  n6 = Constant() {r=0.5}
  n7 = DynamicParameter() {parameterIndex=1, defaultValue=<rgba 0.35 0.25 1 1>, paramNames [Rad,Hardness,Pow,na]}
  n8 = Add(A=@n4, B=Power(Base=SphereMask(A=@n5, B=@n6, Radius=@n7.Rad, Hardness=@n7.Rad), Exp=@n7.Pow))
  n9 = MaterialFunctionCall(
   Fade Length=ScalarParameter() {defaultValue=150, parameterName="Fade Length (S)"}
   Fade Offset=ScalarParameter() {defaultValue=150, parameterName="Fade Offset (S)"}
  ) {materialFunction=&CameraDepthFade}
  DepthFade(Opacity=Saturate(None=Multiply(A=@n8, B=@n9.Result)), FadeDistance=ScalarParameter() {parameterName=FadeDistance})
  DynamicParameter() {parameterIndex=3, paramNames [na,na,na,na]}
  n10 = DynamicParameter() {parameterIndex=2, defaultValue=<rgba 0 0.35 0 1>, paramNames [Color 3 R, Color 3 G, Color 3 B, na]}
  n11 = ParticleColor()
  MaterialFunctionCall(Emissive=LinearInterpolate(A=@n10.RGB, B=LinearInterpolate(A=@n2.RGB, B=@n11.RGB, Alpha=@n3), Alpha=@n4)) {materialFunction=&MF_FX_EmissiveAdjustments}
 ]
}
